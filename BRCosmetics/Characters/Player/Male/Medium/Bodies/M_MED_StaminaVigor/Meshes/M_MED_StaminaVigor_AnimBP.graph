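# Animation Blueprint for the male medium StaminaVigor body cosmetic, making it follow its attached parent mesh with outfit-specific corrections.
# The copied parent pose runs through a deform Control Rig, then additive bone offsets.
# While bIsSkyDiving is true, four back-vest bones shift with a quick blend in and slower blend out; the head always gets a slight roll.
# Spine constraints and other movement-state getters are unconnected, and the update event is empty.
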AnimBlueprint /BRCosmetics/Characters/Player/Male/Medium/Bodies/M_MED_StaminaVigor/Meshes/M_MED_StaminaVigor_AnimBP
target_skeleton=&/M_MED_StaminaVigor_Skeleton
graphs [
 AnimGraph {nodes [
   "Constraint - spine_04 (P)"!(ComponentPose, bAlphaBoolEnabled:bool=True, ConstraintWeights_0=0) -> (Pose) {
    alphaInputType=Bool
    alphaBoolBlend {blendOption=Linear}
    boneToModify {boneName=spine_04}
    constraintSetup [{targetBone {boneName=spine_03}}]
    constraintWeights [1]
   }
   "Constraint - spine_04 (P)"(bAlphaBoolEnabled=True, ConstraintWeights_0=1) {
    alphaInputType=Bool
    alphaBoolBlend {blendOption=Linear}
    boneToModify {boneName=spine_04}
    constraintSetup [{targetBone {boneName=spine_03}}]
    constraintWeights [1]
   }
   "Constraint - spine_05 (R)"(Alpha=0.2, ConstraintWeights_0=1) {
    alphaBoolBlend {blendOption=Linear}
    boneToModify {boneName=spine_05}
    constraintSetup [{transformType=Rotation, targetBone {boneName=spine_04}}]
    constraintWeights [1]
   }
   "Constraint - spine_05 (R)"(Alpha=0.2, ConstraintWeights_0=1) {
    alphaBoolBlend {blendOption=Linear}
    boneToModify {boneName=spine_05}
    constraintSetup [{transformType=Rotation, targetBone {boneName=spine_04}}]
    constraintWeights [1]
   }
   n1 = Control Rig(Source=Copy Pose From Mesh() {bUseAttachedParent=true}) {
    alphaBoolBlend {blendOption=Linear}
    controlRigAssetReference {blueprintRigClass=&/M_MED_StaminaVigor_DeformRig_ControlRig.M_MED_StaminaVigor_DeformRig_ControlRig_C}
   }
   n2 = Get bIsSkyDiving()
   n3 = "Transform (Modify) Bone - Bone: Dyn_BackVest_L"!(ComponentPose=Local To Component(LocalPose=@n1), bAlphaBoolEnabled:bool=@n2, Translation="5,-5,0", Rotation="0,0,0", Scale="1,1,1") -> (Pose) {
    alphaInputType=Bool
    rotationMode=BMM_Additive
    rotationSpace=BCS_BoneSpace
    scaleMode=BMM_Additive
    scaleSpace=BCS_BoneSpace
    translationMode=BMM_Additive
    translationSpace=BCS_BoneSpace
    alphaBoolBlend {blendInTime=0.1, blendOption=Linear, blendOutTime=1}
    boneToModify {boneName=Dyn_BackVest_L}
    translation=<5 -5>
   }
   n4 = "Transform (Modify) Bone - Bone: Dyn_BackVest_R"!(ComponentPose=@n3, bAlphaBoolEnabled:bool=@n2, Translation="5,5,0", Rotation="0,0,0", Scale="1,1,1") -> (Pose) {
    alphaInputType=Bool
    rotationMode=BMM_Additive
    rotationSpace=BCS_BoneSpace
    scaleMode=BMM_Additive
    scaleSpace=BCS_BoneSpace
    translationMode=BMM_Additive
    translationSpace=BCS_BoneSpace
    alphaBoolBlend {blendInTime=0.1, blendOption=Linear, blendOutTime=1}
    boneToModify {boneName=Dyn_BackVest_R}
    translation=<5 5>
   }
   n5 = "Transform (Modify) Bone - Bone: Dyn_BackVest_Lower_L"!(ComponentPose=@n4, bAlphaBoolEnabled:bool=@n2, Translation="0,-5,0", Rotation="0,0,0", Scale="1,1,1") -> (Pose) {
    alphaInputType=Bool
    rotationMode=BMM_Additive
    rotationSpace=BCS_BoneSpace
    scaleMode=BMM_Additive
    scaleSpace=BCS_BoneSpace
    translationMode=BMM_Additive
    translationSpace=BCS_BoneSpace
    alphaBoolBlend {blendInTime=0.1, blendOption=Linear, blendOutTime=1}
    boneToModify {boneName=Dyn_BackVest_Lower_L}
    translation {y=-5}
   }
   n6 = "Transform (Modify) Bone - Bone: Dyn_BackVest_Lower_R"!(ComponentPose=@n5, bAlphaBoolEnabled:bool=@n2, Translation="0,-5,0", Rotation="0,0,0", Scale="1,1,1") -> (Pose) {
    alphaInputType=Bool
    rotationMode=BMM_Additive
    rotationSpace=BCS_BoneSpace
    scaleMode=BMM_Additive
    scaleSpace=BCS_BoneSpace
    translationMode=BMM_Additive
    translationSpace=BCS_BoneSpace
    alphaBoolBlend {blendInTime=0.1, blendOption=Linear, blendOutTime=1}
    boneToModify {boneName=Dyn_BackVest_Lower_R}
    translation {y=-5}
   }
   Output Pose(Result=Component To Local(ComponentPose="Transform (Modify) Bone - Bone: head"(ComponentPose=@n6, Alpha=1, Translation="0,0,0", Rotation="0,0,-5", Scale="1,1,1") {rotationMode=BMM_Additive, alphaBoolBlend {blendOption=Linear}, boneToModify {boneName=head}, rotation {roll=-5}})) {layerGroup=None}
   Get bIsJumping()
   Get bIsFalling()
   Get bIsSkydiveDiveMode()
  ]}
 EventGraph {nodes [Event BlueprintUpdateAnimation()]}
]
variables []
