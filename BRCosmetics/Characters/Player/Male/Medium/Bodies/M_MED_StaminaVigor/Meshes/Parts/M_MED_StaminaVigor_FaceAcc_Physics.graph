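# Physics asset for the FaceAcc mesh of the male medium StaminaVigor player body, setting up dynamic hair bodies.
# Ten tiny per-lock anchor capsules join a master hair body through angularly free constraints, and each carries one larger hair capsule on a tightly limited constraint.
# Main1 and Right2 lock twist and second swing, moving on a single axis.
# An unconstrained right clavicle capsule is also present, apparently as a collider for the hair.
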PhysicsAsset /BRCosmetics/Characters/Player/Male/Medium/Bodies/M_MED_StaminaVigor/Meshes/Parts/M_MED_StaminaVigor_FaceAcc_Physics
bodies {
 Dyn_Hair_Bottom1 [{shapeType=Capsule, length=15, radius=5, shapeName=Dyn_Hair_Bottom1_capsule, center=<-4.4809 0 0>, rotation=<rot 90 0 0>}]
 Dyn_Hair_Bottom2 [
  {shapeType=Capsule, length=13, radius=5, shapeName=Dyn_Hair_Bottom2_capsule, center=<-3.9775 0 0>, rotation=<rot 90 0 0>}
  {shapeType=Capsule, length=12, radius=4, shapeName=Dyn_Hair_Bottom2_capsule, center=<-2.2131 -0.9613 13.766>, rotation=<rot 39.375 -0.0005 0>}
 ]
 Dyn_Hair_Front1 [{shapeType=Capsule, length=15, radius=4, shapeName=Dyn_Hair_Front1_capsule, center=<-3.1915 -2 0>, rotation=<rot 90 0 0>}]
 Dyn_Hair_Left1 [{shapeType=Capsule, length=10, radius=5, shapeName=Dyn_Hair_Left1_capsule, center=<-2.7233 -0.3 0>, rotation=<rot 90 -178.2101 -172.585>}]
 Dyn_Hair_Left2 [{shapeType=Capsule, length=15, radius=5, shapeName=Dyn_Hair_Left2_capsule, center=<-3.5315 0.287 0>, rotation=<rot 90 0 0>}]
 Dyn_Hair_Left3 [{shapeType=Capsule, length=15, radius=5, shapeName=Dyn_Hair_Left3_capsule, center=<-3.946 -0.8363 0>, rotation=<rot 90 0 0>}]
 Dyn_Hair_Main1 [{shapeType=Capsule, length=25, radius=7, shapeName=Dyn_Hair_Main1_capsule, center=<3 0 0>, rotation=<rot 90 0 0>}]
 Dyn_Hair_Right1 [{shapeType=Capsule, length=15, radius=5, shapeName=Dyn_Hair_Right1_capsule, center=<-4.3431 0 0>, rotation=<rot 90 0 0>}]
 Dyn_Hair_Right2 [{shapeType=Capsule, length=15, radius=5, shapeName=Dyn_Hair_Right2_capsule, center=<0.5 0 0>, rotation=<rot 90 0 0>}]
 Dyn_Hair_Right3 [{shapeType=Capsule, length=15, radius=5, shapeName=Dyn_Hair_Right3_capsule, center=<-3.1263 0 0>, rotation=<rot 90 0 0>}]
 Dyn_Master_Hair [{shapeType=Capsule, length=0.5, radius=0.5, shapeName=Dyn_Master_Hair_capsule, rotation=<rot 0 -90 0>}]
 Dyn_Master_HairBottom1 [{shapeType=Capsule, length=0.5, radius=0.5, shapeName=Dyn_Master_HairBottom1_capsule}]
 Dyn_Master_HairBottom2 [{shapeType=Capsule, length=0.5, radius=0.5, shapeName=Dyn_Master_HairBottom2_capsule}]
 Dyn_Master_HairFront1 [{shapeType=Capsule, length=0.5, radius=0.5, shapeName=Dyn_Master_HairFront1_capsule}]
 Dyn_Master_HairLeft1 [{shapeType=Capsule, length=0.5, radius=0.5, shapeName=Dyn_Master_HairLeft1_capsule}]
 Dyn_Master_HairLeft2 [{shapeType=Capsule, length=0.5, radius=0.5, shapeName=Dyn_Master_HairLeft2_capsule}]
 Dyn_Master_HairLeft3 [{shapeType=Capsule, length=0.5, radius=0.5, shapeName=Dyn_Master_HairLeft3_capsule}]
 Dyn_Master_HairMain [{shapeType=Capsule, length=0.5, radius=0.5, shapeName=Dyn_Master_HairMain_capsule}]
 Dyn_Master_HairRight1 [{shapeType=Capsule, length=0.5, radius=0.5, shapeName=Dyn_Master_HairRight1_capsule}]
 Dyn_Master_HairRight2 [{shapeType=Capsule, length=0.5, radius=0.5, shapeName=Dyn_Master_HairRight2_capsule}]
 Dyn_Master_HairRight3 [{shapeType=Capsule, length=0.5, radius=0.5, shapeName=Dyn_Master_HairRight3_capsule}]
 clavicle_r [{shapeType=Capsule, length=20, radius=10, shapeName=clavicle_r_capsule, center=<-13.7227 2.5014 0.2694>, rotation=<rot 67.5 -0.001 -0.0012>}]
}
constraints [
 {
  bone1Name=Dyn_Master_HairMain
  bone2Name=Dyn_Master_Hair
  swing1LimitDegrees=45
  swing1Motion=Free
  swing2LimitDegrees=45
  swing2Motion=Free
  twistLimitDegrees=45
  twistMotion=Free
 }
 {
  bone1Name=Dyn_Master_HairRight1
  bone2Name=Dyn_Master_Hair
  swing1LimitDegrees=45
  swing1Motion=Free
  swing2LimitDegrees=45
  swing2Motion=Free
  twistLimitDegrees=45
  twistMotion=Free
 }
 {
  bone1Name=Dyn_Master_HairRight2
  bone2Name=Dyn_Master_Hair
  swing1LimitDegrees=45
  swing1Motion=Free
  swing2LimitDegrees=45
  swing2Motion=Free
  twistLimitDegrees=45
  twistMotion=Free
 }
 {
  bone1Name=Dyn_Master_HairRight3
  bone2Name=Dyn_Master_Hair
  swing1LimitDegrees=45
  swing1Motion=Free
  swing2LimitDegrees=45
  swing2Motion=Free
  twistLimitDegrees=45
  twistMotion=Free
 }
 {
  bone1Name=Dyn_Master_HairBottom1
  bone2Name=Dyn_Master_Hair
  swing1LimitDegrees=45
  swing1Motion=Free
  swing2LimitDegrees=45
  swing2Motion=Free
  twistLimitDegrees=45
  twistMotion=Free
 }
 {
  bone1Name=Dyn_Master_HairBottom2
  bone2Name=Dyn_Master_Hair
  swing1LimitDegrees=45
  swing1Motion=Free
  swing2LimitDegrees=45
  swing2Motion=Free
  twistLimitDegrees=45
  twistMotion=Free
 }
 {
  bone1Name=Dyn_Master_HairLeft3
  bone2Name=Dyn_Master_Hair
  swing1LimitDegrees=45
  swing1Motion=Free
  swing2LimitDegrees=45
  swing2Motion=Free
  twistLimitDegrees=45
  twistMotion=Free
 }
 {
  bone1Name=Dyn_Master_HairLeft2
  bone2Name=Dyn_Master_Hair
  swing1LimitDegrees=45
  swing1Motion=Free
  swing2LimitDegrees=45
  swing2Motion=Free
  twistLimitDegrees=45
  twistMotion=Free
 }
 {
  bone1Name=Dyn_Master_HairLeft1
  bone2Name=Dyn_Master_Hair
  swing1LimitDegrees=45
  swing1Motion=Free
  swing2LimitDegrees=45
  swing2Motion=Free
  twistLimitDegrees=45
  twistMotion=Free
 }
 {
  bone1Name=Dyn_Master_HairFront1
  bone2Name=Dyn_Master_Hair
  swing1LimitDegrees=45
  swing1Motion=Free
  swing2LimitDegrees=45
  swing2Motion=Free
  twistLimitDegrees=45
  twistMotion=Free
 }
 {
  bone1Name=Dyn_Hair_Main1
  bone2Name=Dyn_Master_HairMain
  swing1LimitDegrees=10
  swing2LimitDegrees=45
  swing2Motion=Locked
  twistLimitDegrees=45
  twistMotion=Locked
 }
 {bone1Name=Dyn_Hair_Right1, bone2Name=Dyn_Master_HairRight1, swing1LimitDegrees=15, swing2LimitDegrees=5, twistLimitDegrees=1}
 {
  bone1Name=Dyn_Hair_Right2
  bone2Name=Dyn_Master_HairRight2
  swing1LimitDegrees=15
  swing2LimitDegrees=45
  swing2Motion=Locked
  twistLimitDegrees=45
  twistMotion=Locked
 }
 {bone1Name=Dyn_Hair_Right3, bone2Name=Dyn_Master_HairRight3, swing1LimitDegrees=10, swing2LimitDegrees=5, twistLimitDegrees=1}
 {bone1Name=Dyn_Hair_Bottom1, bone2Name=Dyn_Master_HairBottom1, swing1LimitDegrees=10, swing2LimitDegrees=1, twistLimitDegrees=1}
 {bone1Name=Dyn_Hair_Bottom2, bone2Name=Dyn_Master_HairBottom2, swing1LimitDegrees=10, swing2LimitDegrees=10, twistLimitDegrees=1}
 {bone1Name=Dyn_Hair_Left3, bone2Name=Dyn_Master_HairLeft3, swing1LimitDegrees=10, swing2LimitDegrees=1, twistLimitDegrees=1}
 {bone1Name=Dyn_Hair_Left2, bone2Name=Dyn_Master_HairLeft2, swing1LimitDegrees=10, swing2LimitDegrees=1, twistLimitDegrees=1}
 {bone1Name=Dyn_Hair_Left1, bone2Name=Dyn_Master_HairLeft1, swing1LimitDegrees=15, swing2LimitDegrees=1, twistLimitDegrees=1}
 {bone1Name=Dyn_Hair_Front1, bone2Name=Dyn_Master_HairFront1, swing1LimitDegrees=10, swing2LimitDegrees=1, twistLimitDegrees=1}
]
