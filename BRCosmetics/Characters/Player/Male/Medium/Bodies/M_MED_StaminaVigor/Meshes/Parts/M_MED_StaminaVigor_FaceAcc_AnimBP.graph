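# Animation Blueprint for the StaminaVigor male character's face accessory mesh, which follows the attached parent mesh's pose and adds secondary hair motion.
# Each update derives the "ADS UP" alpha from the head's reference-transform delta against neck_02.
# That alpha blends an additive rotation and scale offset onto the Dyn_Master_HairLeft4 bone.
# A RigidBody node then simulates the override physics asset in head-bone space with custom gravity and damped component motion.
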AnimBlueprint /BRCosmetics/Characters/Player/Male/Medium/Bodies/M_MED_StaminaVigor/Meshes/Parts/M_MED_StaminaVigor_FaceAcc_AnimBP
target_skeleton=&/M_MED_StaminaVigor_FaceAcc_Skeleton
graphs [
 AnimGraph {nodes [
   n1 = "Transform (Modify) Bone - Bone: Dyn_Master_HairLeft4"(
    ComponentPose=Local To Component(LocalPose=Copy Pose From Mesh() {bUseAttachedParent=true})
    Alpha=Get ADS UP() -> (ADS UP:double)
    Translation="0,0,0"
    Rotation="0,-18,0"
    Scale="0.8,1,1"
   ) {
    rotationMode=BMM_Additive
    rotationSpace=BCS_BoneSpace
    scaleMode=BMM_Additive
    scaleSpace=BCS_BoneSpace
    translationMode=BMM_Additive
    translationSpace=BCS_BoneSpace
    alphaBoolBlend {blendOption=Linear}
    boneToModify {boneName=Dyn_Master_HairLeft4}
    rotation {yaw=-18}
    scale {x=0.8}
   }
   n2 = RigidBody(ComponentPose=@n1) {
    bClampLinearTranslationLimitToRefPose=true
    bFreezeIncomingPoseOnStart=true
    bOverrideWorldGravity=true
    overridePhysicsAsset=&/M_MED_StaminaVigor_FaceAcc_Physics
    simulationSpace=BaseBoneSpace
    alphaBoolBlend {blendOption=Linear}
    baseBoneRef {boneName=head}
    componentAppliedLinearAccClamp=<3000 3000 3000>
    componentLinearAccScale=<0.2 0.2 0.2>
    componentLinearVelScale=<0.2 0.2 0.2>
    overrideWorldGravity {z=-980}
   }
   Output Pose(Result=Component To Local(ComponentPose=@n2)) {layerGroup=None}
  ]}
 EventGraph {nodes [
   Event BlueprintUpdateAnimation()
   > Set ADS UP(ADS UP:double=GetAlphaFromReferenceTransformDelta(
     SkeletalMeshComponent=Get Skeletal Mesh() -> (Skeletal Mesh:Skeletal Mesh Component Object Reference)
     BoneName=head
     BaseName=neck_02
     InputRange=(X=-8,Y=-20)
     TwistAxis="0,0,0"
    )) -> (Output_Get:double)
   Event BlueprintInitializeAnimation()
   > Set Skeletal Mesh(Skeletal Mesh:Skeletal Mesh Component Object Reference=GetOwnerSkeletalMeshComponent()) -> (Output_Get:Skeletal Mesh Component Object Reference)
  ]}
]
variables [
 ADS UP {type=number}
 Skeletal Mesh {type=SkeletalMeshComponent}
]
